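# Physics asset for the Deli Sandwich medium male player body, covering a dangling sticks piece attached at the right thigh.
# Capsules stand in for the pelvis and right thigh twist bone, joined by a constraint with free rotation.
# A tiny master box is rotationally locked to the thigh and anchors the larger sticks box.
# The sticks joint allows moderate swing and twist but is nearly rigid on one swing axis.
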PhysicsAsset /BRCosmetics/Characters/Player/Male/Medium/Bodies/M_MED_Deli_Sandwich/Meshes/M_MED_Deli_Sandwich_Physics
bodies {
 dyn_sticks [{shapeType=Box, extentX=3, extentY=7.7762, extentZ=40, center=<2 -1 0>, rotation=<rot 0 90 -90>}]
 dyn_sticks_master [{shapeType=Box, extentX=0.5, extentY=0.5, extentZ=0.5}]
 pelvis [{shapeType=Capsule, length=10, radius=10, center=<4.875 0.7897 -0.0859>}]
 thigh_twist_01_r [{shapeType=Capsule, length=20, radius=7, center=<-3.017 -0.2457 0.8012>, rotation=<rot 84.3749 0 -2.8126>}]
}
constraints [
 {
  bone1Name=thigh_twist_01_r
  bone2Name=pelvis
  swing1LimitDegrees=45
  swing1Motion=Free
  swing2LimitDegrees=45
  swing2Motion=Free
  twistLimitDegrees=45
  twistMotion=Free
 }
 {
  bone1Name=dyn_sticks_master
  bone2Name=thigh_twist_01_r
  swing1LimitDegrees=45
  swing1Motion=Locked
  swing2LimitDegrees=45
  swing2Motion=Locked
  twistLimitDegrees=45
  twistMotion=Locked
 }
 {bone1Name=dyn_sticks, bone2Name=dyn_sticks_master, swing1LimitDegrees=1, swing2LimitDegrees=30, twistLimitDegrees=30}
]
